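# Face-accessory Control Rig for the female medium Impulse Spring PER_1 skin that poses the hair master bones as the head moves.
# Spherical pose readers measure the head bone relative to spine_05 and output weights.
# These weights drive Modify Transforms offsets on the back, right and left dyn_hair MSTR bones, with right and left sharing one reader.
# A second, similar reader chain is present but not connected to execution.
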ControlRigBlueprint /Game/Characters/Player/Female/Medium/Bodies/F_MED_Impulse_Spring/Skins/PER_01/Meshes/F_MED_Impulse_Spring_PER_1_FaceAcc_CtrlRig
graphs [
 RigVMFunctionLibrary {links [], nodes []}
 RigVMModel {nodes [
   RigUnit_BeginExecution = &RigUnit_BeginExecution()
   n1 = /Script/ControlRig.RigUnit_SequenceExecution(ExecuteContext=@RigUnit_BeginExecution.ExecutePin) {is=struct RigVMUnitNode}
   n2 = &RigUnit_SphericalPoseReader(
    ActiveRegionScaleFactors=(PositiveWidth=8,NegativeWidth=8,PositiveHeight=0.8,NegativeHeight=8)
    ActiveRegionSize=0.05
    Debug=(bDrawDebug=True,DebugScale=25,DebugSegments=20,DebugThickness=0.25)
    DriverAxis=(X=1,Y=0,Z=0)
    DriverItem=(Type=Bone,Name="head")
    FalloffRegionScaleFactors=(PositiveWidth=0.2,NegativeWidth=0.2,PositiveHeight=0.8,NegativeHeight=0.8)
    FalloffSize=0.15
    OptionalParentItem=(Type=Bone,Name="spine_05")
    RotationOffset=(X=-40,Y=-85,Z=25)
    ExecutePin=@n1.A
   )
   > &RigUnit_ModifyTransforms(
    ItemToModify=((Item=(Type=Bone),Transform=(Rotation=(X=0,Y=0,Z=0,W=1),Translation=(X=0,Y=0,Z=0),Scale3D=(X=1,Y=1,Z=1))),(Item=(Type=Bone,Name="dyn_back_hair_MSTR"),Transform=(Rotation=(X=0,Y=0,Z=0,W=1),Translation=(X=-3,Y=0,Z=0),Scale3D=(X=1,Y=1,Z=1))))
    WeightMinimum=0
    ItemToModify.1.Transform.Rotation=&RigVMFunction_MathQuaternionFromEuler(Euler=(X=0,Y=0,Z=-30)).Result
    Weight=@n2.OutputParam
   )
   > n3 = &RigUnit_SphericalPoseReader(
    ActiveRegionScaleFactors=(PositiveWidth=8,NegativeWidth=8,PositiveHeight=0.8,NegativeHeight=4)
    ActiveRegionSize=0.05
    Debug=(bDrawDebug=false,DebugScale=25,DebugSegments=20,DebugThickness=0.25)
    DriverAxis=(X=1,Y=0,Z=0)
    DriverItem=(Type=Bone,Name="head")
    FalloffRegionScaleFactors=(PositiveWidth=0.2,NegativeWidth=0.2,PositiveHeight=0.8,NegativeHeight=0.8)
    FalloffSize=0.15
    OptionalParentItem=(Type=Bone,Name="spine_05")
    RotationOffset=(X=-45,Y=-75,Z=10)
   )
   > &RigUnit_ModifyTransforms(
    ItemToModify=((Item=(Type=Bone),Transform=(Rotation=(X=0,Y=0,Z=0,W=1),Translation=(X=0,Y=0,Z=0),Scale3D=(X=1,Y=1,Z=1))),(Item=(Type=Bone,Name="dyn_right_hair_MSTR"),Transform=(Rotation=(X=0,Y=0,Z=0,W=1),Translation=(X=-3,Y=0,Z=0),Scale3D=(X=1,Y=1,Z=1))))
    WeightMinimum=0
    ItemToModify.1.Transform.Rotation=&RigVMFunction_MathQuaternionFromEuler(Euler=(X=0,Y=20,Z=-20)).Result
    Weight=@n3.OutputParam
   )
   > &RigUnit_ModifyTransforms(
    ItemToModify=((Item=(Type=Bone),Transform=(Rotation=(X=0,Y=0,Z=0,W=1),Translation=(X=0,Y=0,Z=0),Scale3D=(X=1,Y=1,Z=1))),(Item=(Type=Bone,Name="dyn_left_hair_MSTR"),Transform=(Rotation=(X=0,Y=0,Z=0,W=1),Translation=(X=-3,Y=0,Z=0),Scale3D=(X=1,Y=1,Z=1))))
    WeightMinimum=0
    ItemToModify.1.Transform.Rotation=&RigVMFunction_MathQuaternionFromEuler(Euler=(X=0,Y=0,Z=-20)).Result
    Weight=@n3.OutputParam
   )
   n4 = &RigUnit_SphericalPoseReader(
    ActiveRegionScaleFactors=(PositiveWidth=1,NegativeWidth=1,PositiveHeight=1,NegativeHeight=3)
    ActiveRegionSize=0.05
    Debug=(bDrawDebug=True,bDraw2D=false,DebugScale=25,DebugSegments=20,DebugThickness=0.25)
    DriverAxis=(X=1,Y=0,Z=0)
    DriverItem=(Type=Bone,Name="head")
    ExecutePin=()
    FalloffRegionScaleFactors=(PositiveWidth=0.2,NegativeWidth=0.2,PositiveHeight=0.8,NegativeHeight=0.8)
    FalloffSize=0.15
    OptionalParentItem=(Type=Bone,Name="spine_05")
    RotationOffset=(X=0,Y=-55,Z=-10)
   )
   > &RigUnit_ModifyTransforms(
    ItemToModify=((Item=(Type=Bone),Transform=(Rotation=(X=0,Y=0,Z=0,W=1),Translation=(X=0,Y=0,Z=0),Scale3D=(X=1,Y=1,Z=1))),(Item=(Type=Bone,Name="dyn_back_hair_MSTR"),Transform=(Rotation=(X=0,Y=0,Z=0,W=1),Translation=(X=-3,Y=0,Z=0),Scale3D=(X=1,Y=1,Z=1))))
    WeightMinimum=0
    ItemToModify.1.Transform.Rotation=&RigVMFunction_MathQuaternionFromEuler(Euler=(X=0,Y=0,Z=-40)).Result
    Weight=@n4.OutputParam
   )
   > n5 = &RigUnit_SphericalPoseReader(
    ActiveRegionScaleFactors=(PositiveWidth=1,NegativeWidth=1,PositiveHeight=1,NegativeHeight=1)
    ActiveRegionSize=0.05
    Debug=(bDrawDebug=True,bDraw2D=false,DebugScale=25,DebugSegments=20,DebugThickness=0.25)
    DriverAxis=(X=1,Y=0,Z=0)
    DriverItem=(Type=Bone,Name="head")
    FalloffRegionScaleFactors=(PositiveWidth=0.2,NegativeWidth=0.2,PositiveHeight=0.8,NegativeHeight=0.8)
    FalloffSize=0.15
    OptionalParentItem=(Type=Bone,Name="spine_05")
    RotationOffset=(X=0,Y=-55,Z=-10)
   )
   > &RigUnit_ModifyTransforms(
    ItemToModify=((Item=(Type=Bone),Transform=(Rotation=(X=0,Y=0,Z=0,W=1),Translation=(X=0,Y=0,Z=0),Scale3D=(X=1,Y=1,Z=1))),(Item=(Type=Bone,Name="dyn_right_hair_MSTR"),Transform=(Rotation=(X=0,Y=0,Z=0,W=1),Translation=(X=-3,Y=0,Z=0),Scale3D=(X=1,Y=1,Z=1))))
    WeightMinimum=0
    ItemToModify.1.Transform.Rotation=&RigVMFunction_MathQuaternionFromEuler(Euler=(X=0,Y=0,Z=-20)).Result
    Weight=@n5.OutputParam
   )
   > &RigUnit_ModifyTransforms(
    ItemToModify=((Item=(Type=Bone),Transform=(Rotation=(X=0,Y=0,Z=0,W=1),Translation=(X=0,Y=0,Z=0),Scale3D=(X=1,Y=1,Z=1))),(Item=(Type=Bone,Name="dyn_left_hair_MSTR"),Transform=(Rotation=(X=0,Y=0,Z=0,W=1),Translation=(X=-3,Y=0,Z=0),Scale3D=(X=1,Y=1,Z=1))))
    WeightMinimum=0
    ItemToModify.1.Transform.Rotation=&RigVMFunction_MathQuaternionFromEuler(Euler=(X=0,Y=0,Z=-20)).Result
    Weight=@n5.OutputParam
   )
  ]}
]
hierarchy [
 root {loc=<0 0 0>, rot=<0 0 0 1>}
 pelvis {parent=root, loc=<0 -0.5301 98.0571>, rot=<0.0127 0.707 -0.0127 -0.707>}
 spine_01 {parent=pelvis, loc=<6.1208 -0.9328 0>, rot=<0 0 0.034 -0.9994>}
 spine_02 {parent=spine_01, loc=<6.5996 -0.2095 0>, rot=<0 0 0.0238 -0.9997>}
 spine_03 {parent=spine_02, loc=<6.4914 -0.1237 0>, rot=<0 0 0.0149 -0.9999>}
 spine_04 {parent=spine_03, loc=<6.9598 0.0794 0>, rot=<0 0 -0.0132 -0.9999>}
 spine_05 {parent=spine_04, loc=<6.9987 0 0>, rot=<0 0 -0.0015 -1>}
 neck_01 {parent=spine_05, loc=<11.3138 0.1529 0>, rot=<0 0 -0.2003 -0.9797>}
 neck_02 {parent=neck_01, loc=<4.5627 0 0>, rot=<0 0 0.0504 -0.9987>}
 head {parent=neck_02, loc=<5.3914 0 0>, rot=<0 0 0.115 -0.9934>}
 dyn_left_hair_MSTR {parent=head, loc=<9.032 4.3476 -6.5366>, rot=<0.0041 -0.9906 -0.0006 -0.137>}
 dyn_left_hair_01 {parent=dyn_left_hair_MSTR, loc=<0 0 0>, rot=<0 0 0 1>}
 dyn_left_hair_02 {parent=dyn_left_hair_01, loc=<5.1819 -0.0769 0.1364>, rot=<0 -0.2552 0 -0.9669>}
 dyn_left_hair_f {parent=dyn_left_hair_02, loc=<1.9851 1.4606 0.3245>, rot=<0.974 -0.0022 0.0093 0.2263>}
 dyn_left_hair_fs1 {parent=dyn_left_hair_f, loc=<3.7711 -0.1261 0.8018>, rot=<0 0 0 1>}
 dyn_left_hair_fs2 {parent=dyn_left_hair_f, loc=<3.3888 1.5328 -0.8378>, rot=<0 0 0 1>}
 dyn_left_hair_m {parent=dyn_left_hair_02, loc=<6.3552 -1.9272 1.0718>, rot=<0.943 0.0024 -0.0087 -0.3326>}
 dyn_left_hair_b {parent=dyn_left_hair_02, loc=<4.1127 -4.4038 1.6538>, rot=<0.9997 0.0001 -0.0226 0.0047>}
 dyn_left_hair_bs1 {parent=dyn_left_hair_b, loc=<3.747 1.5802 -0.7781>, rot=<0 -0.0296 0 -0.9996>}
 dyn_back_hair_MSTR {parent=head, loc=<8.0626 -9.9146 0.1901>, rot=<-0.0101 -0.9979 0.0611 -0.0214>}
 dyn_back_hair_01 {parent=dyn_back_hair_MSTR, loc=<0 0 0>, rot=<0 0 0 1>}
 dyn_back_hair_02 {parent=dyn_back_hair_01, loc=<4.3256 0 -0.3691>, rot=<0 -0.0738 0 -0.9973>}
 dyn_back_hair_s1 {parent=dyn_back_hair_02, loc=<5.5583 1.0669 1.899>, rot=<0 0 0 1>}
 dyn_back_hair_s2 {parent=dyn_back_hair_02, loc=<4.3299 0.0866 -1.3696>, rot=<0 0 0 1>}
 dyn_back_hair_s3 {parent=dyn_back_hair_02, loc=<5.4639 3.7896 -3.9703>, rot=<0 0 0.0282 -0.9996>}
 dyn_right_hair_MSTR {parent=head, loc=<6.6401 -1.1803 8.1371>, rot=<0.005 -0.9985 0.0515 0.0176>}
 dyn_right_hair_01 {parent=dyn_right_hair_MSTR, loc=<0 0 0>, rot=<0 0 0 1>}
 dyn_right_hair_s1 {parent=dyn_right_hair_01, loc=<7.5379 2.1243 1.492>, rot=<0 0 0 1>}
 dyn_right_hair_s2 {parent=dyn_right_hair_01, loc=<7.4051 -0.208 -0.5682>, rot=<0 0 0 1>}
]
